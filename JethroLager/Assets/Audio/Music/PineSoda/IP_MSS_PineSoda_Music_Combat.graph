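# Stereo MetaSound source for the PineSoda music encounter, playing a combat loop and an ending stinger with distance-based volume.
# A Harmonix metronome retriggers the loop every set number of bars.
# Objective completion, gated to fire once, schedules the stinger on a clock subdivision while the loop fades and stops.
# Pausing fades the loop and finishes the source after a resettable delay; stinger completion also finishes it.
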MetaSoundSource /JethroLager/Assets/Audio/Music/PineSoda/IP_MSS_PineSoda_Music_Combat
inputs [
 EndingStinger {type=WaveAsset}
 Interp Time {type=Time}
 MasterVolume {type=Float}
 Music.BeatsPerBar {type=Int32}
 Music.LoopLengthBars {type=Int32}
 Music.Tempo {type=Float}
 Music.WaveAsset {type=WaveAsset}
 MusicEncounter.Attenuation.Volume.FalloffDistance {type=Float}
 MusicEncounter.Attenuation.Volume.InnerRadius {type=Float}
 MusicEncounter.bIsPaused {type=Bool}
 MusicEncounter.Objective.Complete {type=Trigger}
 MusicEncounter.PauseFadeOutTime {type=Time}
 UE.Attenuation.Distance {type=Float}
 UE.Source.OnPlay {type=Trigger}
]
interfaces [UE.Attenuation:1.0, UE.OutputFormat.Stereo:1.0, UE.Source.OneShot:1.0, UE.Source:1.1]
outputs [
 "UE.OutputFormat.Stereo.Audio:0" {type=Audio}
 "UE.OutputFormat.Stereo.Audio:1" {type=Audio}
 UE.Source.OneShot.OnFinished {type=Trigger}
]
pages [Default {
  input_defaults {
   EndingStinger=/JethroLager/Assets/Audio/Music/PineSoda/SoundWaves/IP_MX_S35_WookieFight_STNG
   Interp Time=0.68
   MasterVolume=1
   Music.BeatsPerBar=4
   Music.LoopLengthBars=5
   Music.Tempo=112
   Music.WaveAsset=/JethroLager/Assets/Audio/Music/PineSoda/SoundWaves/IP_MX_S35_WookieFight_LP
   MusicEncounter.Attenuation.Volume.FalloffDistance=6000
   MusicEncounter.Attenuation.Volume.InnerRadius=2000
   MusicEncounter.Objective.Complete=null
   MusicEncounter.PauseFadeOutTime=2
   MusicEncounter.bIsPaused=null
   UE.Attenuation.Distance=0
   UE.Source.OnPlay=null
  }
  nodes [
   TriggerOnValueChange_0 = UE.Trigger On Value Change:Bool(Value=^MusicEncounter.bIsPaused)
   n1 = TriggerCompare.Clamp:Bool(A=^MusicEncounter.bIsPaused, Compare=@TriggerOnValueChange_0.Trigger) {also {defaults {B="<MetaSoundBuilderResult.SUCCEEDED: 0>"}}}
   n2 = Value.Value:Float(Reset=@n1.False, Set=@n1.True) {also {defaults {Init Value=1, Target Value=0}}}
   n3 = UE.InterpTo:Audio(Interp Time=^MusicEncounter.PauseFadeOutTime, Target=@n2.Output Value)
   n4 = VariableDeferredAccessor.Trigger(Variable=InitVariable.Trigger().Variable)
   TriggerControl_0 = UE.Trigger Control(Trigger In=^MusicEncounter.Objective.Complete, Close=@n4.Value)
   n5 = HarmonixNodes.Metronome(Time Sig. Numerator=^Music.BeatsPerBar, Tempo=^Music.Tempo, Transport=HarmonixNodes.TriggerToTransport(Play=^UE.Source.OnPlay).Transport)
   n6 = HarmonixNodes.MidiClockSubdivisionTrigger(MIDI Clock=@n5.MIDI Clock)
   n7 = InitVariable.Trigger()
   n8 = &MSP_ScheduleTrigger(
    Schedule=@TriggerControl_0.Trigger Out
    In Trigger=VariableAccessor.Trigger(Variable=VariableMutator.Trigger(Value=@n6.Trigger Out, Variable=@n7.Variable).Variable).Value
   )
   n9 = Value.Value:Float(Set=@n8.Scheduled Trigger) {also {defaults {Init Value=1, Target Value=0}}}
   n10 = VariableMutator.Float(Variable=InitVariable.Float().Variable, Value=UE.InterpTo:Audio(Interp Time=^Interp Time, Target=@n9.Output Value).Value)
   n11 = InitVariable.Trigger()
   TriggerDelay_1 = UE.Trigger Delay(Delay Time=^Interp Time, In=@n9.On Set)
   n12 = HarmonixNodes.MidiClockSubdivisionTrigger(Grid Size Multiplier=^Music.LoopLengthBars, MIDI Clock=@n5.MIDI Clock) {also {defaults {Base Grid Size=0}}}
   n13 = InitVariable.Trigger()
   WavePlayer_0 = UE.Wave Player:Stereo(
    Wave Asset=^Music.WaveAsset
    Stop=VariableAccessor.Trigger(Variable=VariableMutator.Trigger(Variable=@n11.Variable, Value=@TriggerDelay_1.Out).Variable).Value
    Play=VariableAccessor.Trigger(Variable=VariableMutator.Trigger(Value=@n12.Trigger Out, Variable=@n13.Variable).Variable).Value
   ) {also {version [1,1]}}
   WavePlayer_1 = UE.Wave Player:Stereo(Wave Asset=^EndingStinger, Play=@n8.Scheduled Trigger) {also {version [1,1]}}
   AudioMixerAudioMixerStereo2_0 = "AudioMixer.Audio Mixer (Stereo, 2)"(
    Gain 0=UE.Multiply:Float(
     PrimaryOperand=UE.Multiply:Float(PrimaryOperand=^MasterVolume, AdditionalOperands=@n3.Value).Out
     AdditionalOperands=VariableAccessor.Float(Variable=@n10.Variable).Value
    ).Out
    In 0 L=@WavePlayer_0.Out Left
    In 0 R=@WavePlayer_0.Out Right
    In 1 L=@WavePlayer_1.Out Left
    In 1 R=@WavePlayer_1.Out Right
   )
   ADSREnvelopeADSREnvelope_0 = ADSR Envelope.ADSR Envelope:Float(Trigger Attack=^UE.Source.OnPlay) {also {defaults {Attack Time=2, Decay Time=0, Sustain Level=1}}}
   n14 = UE.Add:Float(AdditionalOperands=^MusicEncounter.Attenuation.Volume.FalloffDistance, PrimaryOperand=^MusicEncounter.Attenuation.Volume.InnerRadius)
   n15 = UE.InterpTo:Audio(Target=^UE.Attenuation.Distance) {also {defaults {Interp Time=0.5}}}
   n16 = UE.Multiply:Float(
    AdditionalOperands=@ADSREnvelopeADSREnvelope_0.Out Envelope
    PrimaryOperand=UE.Multiply:Float(AdditionalOperands=MapRange.MapRange:Float(In Range A=^MusicEncounter.Attenuation.Volume.InnerRadius, In Range B=@n14.Out, In=@n15.Value) {also {defaults {Out Range A=1, Out Range B=0}}}.Out Value).Out
   )
   Multiply_0 = UE.Multiply:Audio by Float(PrimaryOperand=@AudioMixerAudioMixerStereo2_0.Out L, AdditionalOperands=@n16.Out)
   Multiply_1 = UE.Multiply:Audio by Float(PrimaryOperand=@AudioMixerAudioMixerStereo2_0.Out R, AdditionalOperands=@n16.Out)
   TriggerDelay_0 = UE.Trigger Delay(Delay Time=^MusicEncounter.PauseFadeOutTime, Reset=@n2.On Reset, In=@n2.On Set)
   TriggerAnyTriggerAny2_0 = "TriggerAny.Trigger Any (2)"(In 0=@TriggerDelay_0.Out, In 1=@WavePlayer_1.On Finished)
   VariableMutator.Trigger(Value=@TriggerControl_0.Trigger Out, Variable=@n4.Variable)
  ]
  links [
   {dst="@UE.OutputFormat.Stereo.Audio:0", src=Multiply_0.Out}
   {dst="@UE.OutputFormat.Stereo.Audio:1", src=Multiply_1.Out}
   {dst="@UE.Source.OneShot.OnFinished", src=TriggerAnyTriggerAny2_0.Out}
  ]
 }]
properties {
 assetImportData=&self:AssetImportData
 numChannels=2
 outputFormat=Stereo
 sampleRate=48000
 soundClassObject=&Fort_Music_NPC
}
